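# UI-domain unlit material in the TextureHeatmapRenderer editor utility, named as the dot for texture heatmaps.
# A RadialGradientExponential function forms a soft radial falloff, with its Density input exposed as a scalar parameter.
# A constant color vector, left at its default, sits alongside the gradient.
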Material /Game/NeverCook/EditorUtilities/TextureHeatmapRenderer/Art/M_TextureHeatmapDot
graph {
 properties {materialDomain=MD_UI, pixelDepthOffsetMode=PDOM_Legacy, shadingModel=MSM_Unlit, usedShadingModels=Unlit}
 nodes [
  Constant3Vector()
  MaterialFunctionCall(Density=ScalarParameter() {defaultValue=2, parameterName=Density}) {materialFunction=&RadialGradientExponential}
 ]
}
